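# Physics asset for the dynamic jacket on the male medium Bites Jacket cosmetic body, giving the jacket secondary motion.
# Seven chains (back, sides, front and back corners) each link a tiny master body to two capsules with tight swing and twist limits.
# Masters attach to spine_02 with free rotation, and free-rotation constraints tie the lower left-back and right-back segments to the center back.
# Pelvis, spine and thigh capsules supply the body volumes around the jacket.
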PhysicsAsset /BRCosmetics/Characters/Player/Male/Medium/Bodies/M_MED_Bites_Jacket/Meshes/M_MED_Bites_Jacket_Dyn_Jacket
bodies {
 dyn_jacket_bk_01 [{shapeType=Capsule, length=12.6277, radius=9.3022, shapeName=dyn_jacket_l_bk_01_capsule, center=<9 0 5>, rotation=<rot 0 0 90>}]
 dyn_jacket_bk_02 [{shapeType=Capsule, length=13.2182, radius=8.7116, shapeName=dyn_jacket_l_bk_02_capsule, center=<9 0 5>, rotation=<rot 0 0 90>}]
 dyn_jacket_bk_MSTR [{shapeType=Capsule, length=1, radius=1}]
 dyn_jacket_l_01 [{shapeType=Capsule, length=9.8303, radius=6.2915, shapeName=dyn_jacket_l_01_capsule, center=<7 0 2.5>, rotation=<rot 0 0 90>}]
 dyn_jacket_l_02 [{shapeType=Capsule, length=9.8303, radius=6.2915, shapeName=dyn_jacket_l_02_capsule, center=<7 0 2.5>, rotation=<rot 0 0 90>}]
 dyn_jacket_l_MSTR [{shapeType=Capsule, length=1, radius=1}]
 dyn_jacket_l_bk_01 [{shapeType=Capsule, length=9.8303, radius=6.2915, shapeName=dyn_jacket_l_bk_01_capsule, center=<7 0 2.5>, rotation=<rot 0 0 90>}]
 dyn_jacket_l_bk_02 [{shapeType=Capsule, length=9.8303, radius=6.2915, shapeName=dyn_jacket_l_bk_02_capsule, center=<7 0 2.5>, rotation=<rot 0 0 90>}]
 dyn_jacket_l_bk_MSTR [{shapeType=Capsule, length=1, radius=1}]
 dyn_jacket_l_fr_01 [{shapeType=Capsule, length=5.0791, radius=7.0742, shapeName=dyn_jacket_l_fr_01_capsule, center=<10 0 5>, rotation=<rot 90 0 0>}]
 dyn_jacket_l_fr_02 [{shapeType=Capsule, length=7.5383, radius=7.0742, shapeName=dyn_jacket_l_fr_02_capsule, center=<10 0 5>, rotation=<rot 90 0 0>}]
 dyn_jacket_l_fr_MSTR [{shapeType=Capsule, length=1, radius=1}]
 dyn_jacket_r_01 [{shapeType=Capsule, length=9.8303, radius=6.2915, shapeName=dyn_jacket_l_01_capsule, center=<7 0 2.5>, rotation=<rot 0 0 90>}]
 dyn_jacket_r_02 [{shapeType=Capsule, length=9.8303, radius=6.2915, shapeName=dyn_jacket_l_02_capsule, center=<7 0 2.5>, rotation=<rot 0 0 90>}]
 dyn_jacket_r_MSTR [{shapeType=Capsule, length=1, radius=1}]
 dyn_jacket_r_bk_01 [{shapeType=Capsule, length=9.8303, radius=6.2915, shapeName=dyn_jacket_l_bk_01_capsule, center=<7 0 2.5>, rotation=<rot 0 0 90>}]
 dyn_jacket_r_bk_02 [{shapeType=Capsule, length=9.8303, radius=6.2915, shapeName=dyn_jacket_l_bk_02_capsule, center=<7 0 2.5>, rotation=<rot 0 0 90>}]
 dyn_jacket_r_bk_MSTR [{shapeType=Capsule, length=1, radius=1}]
 dyn_jacket_r_fr_01 [{shapeType=Capsule, length=5.0791, radius=7.0742, shapeName=dyn_jacket_l_fr_01_capsule, center=<10 0 5>, rotation=<rot 90 0 0>}]
 dyn_jacket_r_fr_02 [{shapeType=Capsule, length=7.5383, radius=7.0742, shapeName=dyn_jacket_l_fr_02_capsule, center=<10 0 5>, rotation=<rot 90 0 0>}]
 dyn_jacket_r_fr_MSTR [{shapeType=Capsule, length=1, radius=1}]
 pelvis [{shapeType=Capsule, length=5.225, radius=10.4622, center=<-4.1085 0.5108 0>}]
 spine_01 [{shapeType=Capsule, length=5.6506, radius=10.4006, center=<0 0.9137 0>}]
 spine_02 [{shapeType=Capsule, length=5.6506, radius=10.4006, center=<0 0.6443 0>}]
 thigh_twist_01_l [{shapeType=Capsule, length=24.2032, radius=8.5288, center=<8.0756 1.1502 2.1543>, rotation=<rot 86.8255 2.7177 -176.97>}]
 thigh_twist_01_r [{shapeType=Capsule, length=23.6397, radius=9.0923, center=<-8.0431 -0.8172 -1.7254>, rotation=<rot 88.1624 174.46 -5.2356>}]
}
constraints [
 {bone1Name=dyn_jacket_r_bk_01, bone2Name=dyn_jacket_r_bk_MSTR, swing1LimitDegrees=7, swing2LimitDegrees=25, twistLimitDegrees=2}
 {bone1Name=dyn_jacket_r_bk_02, bone2Name=dyn_jacket_r_bk_01, swing1LimitDegrees=7, swing2LimitDegrees=25, twistLimitDegrees=2}
 {bone1Name=dyn_jacket_l_bk_01, bone2Name=dyn_jacket_l_bk_MSTR, swing1LimitDegrees=7, swing2LimitDegrees=25, twistLimitDegrees=2}
 {bone1Name=dyn_jacket_l_bk_02, bone2Name=dyn_jacket_l_bk_01, swing1LimitDegrees=7, swing2LimitDegrees=25, twistLimitDegrees=2}
 {bone1Name=dyn_jacket_l_fr_01, bone2Name=dyn_jacket_l_fr_MSTR, swing1LimitDegrees=7, swing2LimitDegrees=25, twistLimitDegrees=2}
 {bone1Name=dyn_jacket_l_fr_02, bone2Name=dyn_jacket_l_fr_01, swing1LimitDegrees=7, swing2LimitDegrees=25, twistLimitDegrees=2}
 {bone1Name=dyn_jacket_r_fr_01, bone2Name=dyn_jacket_r_fr_MSTR, swing1LimitDegrees=7, swing2LimitDegrees=25, twistLimitDegrees=2}
 {bone1Name=dyn_jacket_r_fr_02, bone2Name=dyn_jacket_r_fr_01, swing1LimitDegrees=7, swing2LimitDegrees=25, twistLimitDegrees=2}
 {bone1Name=dyn_jacket_l_02, bone2Name=dyn_jacket_l_01, swing1LimitDegrees=7, swing2LimitDegrees=25, twistLimitDegrees=2}
 {bone1Name=dyn_jacket_l_01, bone2Name=dyn_jacket_l_MSTR, swing1LimitDegrees=7, swing2LimitDegrees=25, twistLimitDegrees=2}
 {bone1Name=dyn_jacket_r_01, bone2Name=dyn_jacket_r_MSTR, swing1LimitDegrees=7, swing2LimitDegrees=25, twistLimitDegrees=2}
 {bone1Name=dyn_jacket_r_02, bone2Name=dyn_jacket_r_01, swing1LimitDegrees=7, swing2LimitDegrees=25, twistLimitDegrees=2}
 {
  bone1Name=dyn_jacket_r_bk_MSTR
  bone2Name=spine_02
  swing1LimitDegrees=45
  swing1Motion=Free
  swing2LimitDegrees=45
  swing2Motion=Free
  twistLimitDegrees=45
  twistMotion=Free
 }
 {
  bone1Name=dyn_jacket_l_bk_MSTR
  bone2Name=spine_02
  swing1LimitDegrees=45
  swing1Motion=Free
  swing2LimitDegrees=45
  swing2Motion=Free
  twistLimitDegrees=45
  twistMotion=Free
 }
 {
  bone1Name=dyn_jacket_bk_MSTR
  bone2Name=spine_02
  swing1LimitDegrees=45
  swing1Motion=Free
  swing2LimitDegrees=45
  swing2Motion=Free
  twistLimitDegrees=45
  twistMotion=Free
 }
 {
  bone1Name=dyn_jacket_r_MSTR
  bone2Name=spine_02
  swing1LimitDegrees=45
  swing1Motion=Free
  swing2LimitDegrees=45
  swing2Motion=Free
  twistLimitDegrees=45
  twistMotion=Free
 }
 {
  bone1Name=dyn_jacket_l_MSTR
  bone2Name=spine_02
  swing1LimitDegrees=45
  swing1Motion=Free
  swing2LimitDegrees=45
  swing2Motion=Free
  twistLimitDegrees=45
  twistMotion=Free
 }
 {
  bone1Name=dyn_jacket_l_fr_MSTR
  bone2Name=spine_02
  swing1LimitDegrees=45
  swing1Motion=Free
  swing2LimitDegrees=45
  swing2Motion=Free
  twistLimitDegrees=45
  twistMotion=Free
 }
 {
  bone1Name=dyn_jacket_r_fr_MSTR
  bone2Name=spine_02
  swing1LimitDegrees=45
  swing1Motion=Free
  swing2LimitDegrees=45
  swing2Motion=Free
  twistLimitDegrees=45
  twistMotion=Free
 }
 {
  bone1Name=thigh_twist_01_l
  bone2Name=pelvis
  swing1LimitDegrees=45
  swing1Motion=Free
  swing2LimitDegrees=45
  swing2Motion=Free
  twistLimitDegrees=45
  twistMotion=Free
 }
 {
  bone1Name=thigh_twist_01_r
  bone2Name=pelvis
  swing1LimitDegrees=45
  swing1Motion=Free
  swing2LimitDegrees=45
  swing2Motion=Free
  twistLimitDegrees=45
  twistMotion=Free
 }
 {bone1Name=dyn_jacket_bk_01, bone2Name=dyn_jacket_bk_MSTR, swing1LimitDegrees=7, swing2LimitDegrees=25, twistLimitDegrees=2}
 {bone1Name=dyn_jacket_bk_02, bone2Name=dyn_jacket_bk_01, swing1LimitDegrees=7, swing2LimitDegrees=25, twistLimitDegrees=2}
 {
  bone1Name=dyn_jacket_l_bk_02
  bone2Name=dyn_jacket_bk_02
  swing1LimitDegrees=45
  swing1Motion=Free
  swing2LimitDegrees=45
  swing2Motion=Free
  twistLimitDegrees=45
  twistMotion=Free
 }
 {
  bone1Name=dyn_jacket_r_bk_02
  bone2Name=dyn_jacket_bk_02
  swing1LimitDegrees=45
  swing1Motion=Free
  swing2LimitDegrees=45
  swing2Motion=Free
  twistLimitDegrees=45
  twistMotion=Free
 }
 {
  bone1Name=spine_01
  bone2Name=pelvis
  swing1LimitDegrees=45
  swing1Motion=Free
  swing2LimitDegrees=45
  swing2Motion=Free
  twistLimitDegrees=45
  twistMotion=Free
 }
 {
  bone1Name=spine_02
  bone2Name=spine_01
  swing1LimitDegrees=45
  swing1Motion=Free
  swing2LimitDegrees=45
  swing2Motion=Free
  twistLimitDegrees=45
  twistMotion=Free
 }
]
