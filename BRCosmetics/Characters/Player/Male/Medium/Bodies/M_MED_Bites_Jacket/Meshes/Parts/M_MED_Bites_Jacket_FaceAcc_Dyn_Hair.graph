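# Physics asset for the dynamic hair on the face accessory part of the male medium Bites Jacket character body.
# Capsule bodies cover left, right and front hair bones, plus one capsule on the head.
# Strands are one or two bodies long, with each root constrained to the head and each second segment to its first.
# Narrow swing and twist limits keep the hair movement close to its rest shape.
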PhysicsAsset /BRCosmetics/Characters/Player/Male/Medium/Bodies/M_MED_Bites_Jacket/Meshes/Parts/M_MED_Bites_Jacket_FaceAcc_Dyn_Hair
bodies {
 dyn_hair_a_l_01 [{shapeType=Capsule, length=3.3598, radius=2.8332, shapeName=dyn_hair_02_capsule, center=<3 0 2>, rotation=<rot 90 0 0>}]
 dyn_hair_a_r_01 [{shapeType=Capsule, length=2.3598, radius=3.5832, shapeName=dyn_hair_02_capsule, center=<4 0 2>, rotation=<rot 70.0006 0.0013 0.0014>}]
 dyn_hair_a_r_02 [{shapeType=Capsule, length=5.8598, radius=2.8332, shapeName=dyn_hair_02_capsule, center=<5.6 0 0>, rotation=<rot 90 0 0>}]
 dyn_hair_b_l_01 [{shapeType=Capsule, length=2.7348, radius=3.4582, shapeName=dyn_hair_02_capsule, center=<5 0 2>, rotation=<rot 85 -0.0003 -0.0005>}]
 dyn_hair_b_r_01 [{shapeType=Capsule, length=2.3598, radius=4.8332, shapeName=dyn_hair_02_capsule, center=<5 0 2>, rotation=<rot 90 0 0>}]
 dyn_hair_b_r_02 [{shapeType=Capsule, length=3.4848, radius=3.7082, shapeName=dyn_hair_02_capsule, center=<6 0 0>, rotation=<rot 65 0 0>}]
 dyn_hair_c_l_01 [{shapeType=Capsule, length=3.1532, radius=3.0397, shapeName=dyn_hair_02_capsule, center=<4 0 2>, rotation=<rot 90 0 0>}]
 dyn_hair_c_r_01 [{shapeType=Capsule, length=3.7348, radius=3.4582, shapeName=dyn_hair_02_capsule, center=<5 0 0>, rotation=<rot 0 0 100.0001>}]
 dyn_hair_c_r_02 [{shapeType=Capsule, length=3.6098, radius=3.5832, shapeName=dyn_hair_02_capsule, center=<5 0 2>, rotation=<rot 90 0 0>}]
 dyn_hair_d_r_01 [{shapeType=Capsule, length=3.6098, radius=1.9582, shapeName=dyn_hair_02_capsule, center=<3 0 1>, rotation=<rot 90 0 0>}]
 dyn_hair_d_r_02 [{shapeType=Capsule, length=5.9848, radius=2.2082, shapeName=dyn_hair_02_capsule, center=<5 0 0>, rotation=<rot 80.0001 -0.0003 0.0001>}]
 dyn_hair_e_r_01 [{shapeType=Capsule, length=3.3598, radius=3.8332, shapeName=dyn_hair_02_capsule, center=<5 0 2>, rotation=<rot 90 0 0>}]
 dyn_hair_ft_a_01 [{shapeType=Capsule, length=3.3598, radius=2.2082, shapeName=dyn_hair_02_capsule, center=<3 0 1.5>, rotation=<rot 85.0001 0.0003 0.0003>}]
 dyn_hair_ft_a_02 [{shapeType=Capsule, length=3.6098, radius=2.5832, shapeName=dyn_hair_02_capsule, center=<4 0 1.5>, rotation=<rot 90 0 0>}]
 head [{shapeType=Capsule, length=3.9331, radius=6.9992, center=<5.8343 2.0803 0>, rotation=<rot 0 -0.0006 90>}]
}
constraints [
 {bone1Name=dyn_hair_a_l_01, bone2Name=head, swing1LimitDegrees=15, swing2LimitDegrees=30, twistLimitDegrees=5}
 {bone1Name=dyn_hair_a_r_02, bone2Name=dyn_hair_a_r_01, swing1LimitDegrees=15, swing2LimitDegrees=20, twistLimitDegrees=2}
 {bone1Name=dyn_hair_b_l_01, bone2Name=head, swing1LimitDegrees=15, swing2LimitDegrees=30, twistLimitDegrees=5}
 {bone1Name=dyn_hair_b_r_01, bone2Name=head, swing1LimitDegrees=10, swing2LimitDegrees=30, twistLimitDegrees=5}
 {bone1Name=dyn_hair_b_r_02, bone2Name=dyn_hair_b_r_01, swing1LimitDegrees=10, swing2LimitDegrees=30, twistLimitDegrees=5}
 {bone1Name=dyn_hair_c_l_01, bone2Name=head, swing1LimitDegrees=5, swing2LimitDegrees=30, twistLimitDegrees=5}
 {bone1Name=dyn_hair_c_r_01, bone2Name=head, swing1LimitDegrees=5, swing2LimitDegrees=30, twistLimitDegrees=5}
 {bone1Name=dyn_hair_ft_a_01, bone2Name=head, swing1LimitDegrees=5, swing2LimitDegrees=30, twistLimitDegrees=2}
 {bone1Name=dyn_hair_ft_a_02, bone2Name=dyn_hair_ft_a_01, swing1LimitDegrees=10, swing2LimitDegrees=30, twistLimitDegrees=2}
 {bone1Name=dyn_hair_d_r_02, bone2Name=dyn_hair_d_r_01, swing1LimitDegrees=10, swing2LimitDegrees=15, twistLimitDegrees=5}
 {bone1Name=dyn_hair_d_r_01, bone2Name=head, swing1LimitDegrees=10, swing2LimitDegrees=15, twistLimitDegrees=5}
 {bone1Name=dyn_hair_e_r_01, bone2Name=head, swing1LimitDegrees=15, swing2LimitDegrees=30, twistLimitDegrees=5}
 {bone1Name=dyn_hair_c_r_02, bone2Name=dyn_hair_c_r_01, swing1LimitDegrees=15, swing2LimitDegrees=30, twistLimitDegrees=5}
 {bone1Name=dyn_hair_a_r_01, bone2Name=head, swing1LimitDegrees=10, swing2LimitDegrees=20, twistLimitDegrees=2}
]
